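# Sound cue named for the siren of the Athena SpeedSign quest, set up as an in-game mission sound.
# The node graph is empty with no root node, so the cue references no audio to play.
# It carries only playback settings: gadget-activate attenuation, the mission in-game sound class, a boosted volume and a maximum audible distance.
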SoundCue /Game/Athena/Sounds/Quests/SpeedSign/Athena_Qiuest_SpeedSign_Siren_Cue
root=null
nodes []
properties {
 attenuationSettings=&Fort_GadgetActivate_Att
 maxDistance=4000
 soundClassObject=/Game/Sounds/FortSoundClassesAndModes/Fort_Mission_IngameSounds
 volumeMultiplier=2
}
